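# Sound cue for the Section 02 "Phew" impact in the PistolCanary event, built from tentacle attack projectile impact sounds.
# A Random node picks one of three equally weighted impact wave variations on each play.
# Output plays at reduced volume through the Fort_Events sound class, using the tentacle attenuation asset.
# The projectile concurrency asset governs simultaneous instances, and the concurrency set also holds an empty entry.
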SoundCue /PistolCanary_Event_Audio/Assets/Section02/Impacts/SC_PistolCanary_Section02_Phew_Impact
root=Random_0
nodes [
 {id=Random_0, children [WavePlayer_0,WavePlayer_2,WavePlayer_3], properties {weights [1,1,1]}}
 {
  id=WavePlayer_0
  properties {soundWaveAssetPtr=/PistolCanary_Event_Audio/Assets/Section02/wav/PistolCanary_Tentacle_Attk_4000_Proj_Impact_03}
 }
 {
  id=WavePlayer_2
  properties {soundWaveAssetPtr=/PistolCanary_Event_Audio/Assets/Section02/wav/PistolCanary_Tentacle_Attk_4000_Proj_Impact_02}
 }
 {
  id=WavePlayer_3
  properties {soundWaveAssetPtr=/PistolCanary_Event_Audio/Assets/Section02/wav/PistolCanary_Tentacle_Attk_4000_Proj_Impact_01}
 }
]
properties {
 attenuationSettings=/PistolCanary_Event_Audio/Assets/Settings/ATT_PistolCanary_Tentacle
 maxDistance=23000
 soundClassObject=&Fort_Events
 volumeMultiplier=0.4
 concurrencySet [/PistolCanary_Event_Audio/Assets/Settings/CONC_PistolCanary_Projectiles, null]
}
